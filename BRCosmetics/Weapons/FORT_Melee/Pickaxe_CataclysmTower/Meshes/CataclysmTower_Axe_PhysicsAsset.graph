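# Physics asset for the CataclysmTower pickaxe skeletal mesh, a two-handed melee weapon cosmetic.
# It holds a single capsule body on the Melee_TwoHanded_Handle bone, offset and rotated along the handle.
# With no constraints, it gives the axe one simple collision shape with no jointed simulation between bones.
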PhysicsAsset /BRCosmetics/Weapons/FORT_Melee/Pickaxe_CataclysmTower/Meshes/CataclysmTower_Axe_PhysicsAsset
bodies {Melee_TwoHanded_Handle [{shapeType=Capsule, length=46.4511, radius=7.5127, center=<-0.4217 0.2952 30.3578>, rotation=<rot -1.5652 -90 0.0016>}]}
constraints []
